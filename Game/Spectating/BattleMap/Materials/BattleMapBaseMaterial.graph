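# BattleMapBaseMaterial is a Default Lit material for the spectating battle map, enabled for Nanite and spline meshes.
# A parameterised Fresnel term, scaled and raised to a power, drives the rim outline.
# Below PlayeroutlineStrength it blends Colour toward OutlineColour; above it shows PlayerColour, with a thin OutlineDivider band at the threshold.
# A second threshold branch isolates the PlayerColour rim against a constant colour.
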Material /Game/Spectating/BattleMap/Materials/BattleMapBaseMaterial
graph {
 properties {bUsedWithNanite=true, bUsedWithSplineMeshes=true, pixelDepthOffsetMode=PDOM_Legacy, usedShadingModels=Default Lit}
 nodes [
  n1 = ScalarParameter() {defaultValue=2, parameterName=OutlineExponent}
  n2 = Power(
   Base=Multiply(A=Fresnel(ExponentIn=@n1) {baseReflectFraction=0.01}, B=ScalarParameter() {parameterName=OutlineStrengthMultiplier})
   Exp=ScalarParameter() {defaultValue=1.2, parameterName=OutlineStrengthExponent}
  )
  n3 = ScalarParameter() {defaultValue=0.1, parameterName=PlayeroutlineStrength}
  n4 = VectorParameter() {parameterName=PlayerColour}
  n5 = If(
   A=@n2
   B=@n3
   "A > B"=@n4.RGB
   "A == B"=@n4.RGB
   "A < B"=LinearInterpolate(
    A=VectorParameter() {parameterName=Colour, defaultValue=<rgba 0.6719 0.6601 0.028 0>}.RGB
    B=VectorParameter() {parameterName=OutlineColour}.RGB
    Alpha=@n2
   )
  ) {constB=0.6}
  If(A=@n2, B=@n3, "A > B"=@n5, "A == B"=VectorParameter() {parameterName=OutlineDivider}.RGB, "A < B"=@n5) {constB=0.6, equalsThreshold=0.025}
  If(A=@n2, B=@n3, "A > B"=@n4.RGB, "A < B"=Constant3Vector())
 ]
}
